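# Splitscreen lobby prompt widget showing a player avatar, a player label and an input action icon with "Press to join".
# A FortVisualAttachment root holds a grid with the circular-backed avatar layered above a dark info panel.
# GetPrimaryText formats a localized "Player {##}" from a player index, and SetBindingHandle passes a UI action binding handle to the CommonActionWidget.
# The event graph has no logic, and nothing here calls either function.
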WidgetBlueprint /Game/UI/Frontend/Lobby/Splitscreen/WBP_Splitscreen_Prompt_Join
graphs [
 EventGraph {nodes [
   Event PreConstruct()
   Event Construct()
   Event Tick()
  ]}
 GetPrimaryText {nodes [
   n1 = GetPrimaryText!() -> (then, PlayerIndex:int) {is=title K2Node_FunctionEntry}
   > Return Node!(execute, Text:Text=Format Text!(Format="LOCTEXT(\"Player {##}\")", "##":int=@n1) -> (Result))
  ]}
 SetBindingHandle {nodes [
   n1 = SetBindingHandle!() -> (then, NewParam:UI Action Binding Handle Structure) {is=title K2Node_FunctionEntry}
   > SetInputActionBinding!(
    execute
    self:Common Action Widget Object Reference=Get ActionWidget() -> (ActionWidget:Common Action Widget Object Reference)
    BindingHandle:UI Action Binding Handle Structure=@n1
   ) -> (then)
  ]}
]
variables []
widgets {
 info {inheritedWidgetCount=0, parentClass=&UserWidget, rootWidgetClass=&FortVisualAttachment, widgetCount=12}
 widgets [
  FortVisualAttachment_69 {widgetClassPath=&FortVisualAttachment, widget=FortVisualAttachment_69, properties {contentAnchor=<0.5 1.5>}, slot=~}
  GridPanel_JoinPrompt {
   widgetClassPath=&GridPanel
   parent=FortVisualAttachment_69
   widget=GridPanel_JoinPrompt
   properties {columnFill [1], rowFill [0,0]}
   slot {class=SizeBoxSlot}
  }
  Overlay_PlayerAvatar {
   widgetClassPath=&Overlay
   bIsVariable=true
   parent=GridPanel_JoinPrompt
   widget=Overlay_PlayerAvatar
   slot {class=GridSlot, properties {horizontalAlignment=HAlign_Center, layer=10, verticalAlignment=VAlign_Top, nudge=<0 8>}}
  }
  WBP_UIKit_Backplate {
   widgetClassPath=&WBP_UIKit_Backplate
   bIsVariable=true
   parent=Overlay_PlayerAvatar
   widget=WBP_UIKit_Backplate
   properties {brightness=Dark, cornerRadius=Circle, mobileCorners {backplate=25, button=0.3, circle=0.5, interactive Component=20, tab=24}}
   slot {class=OverlaySlot, properties {horizontalAlignment=HAlign_Fill, verticalAlignment=VAlign_Fill}}
  }
  WBP_PlayerAvatar {
   widgetClassPath=/Game/UI/Foundation/Profile/WBP_PlayerAvatar
   bIsVariable=true
   parent=Overlay_PlayerAvatar
   widget=WBP_PlayerAvatar
   properties {showOnlinePresence=false, avatarSize=<51 51>}
   slot {class=OverlaySlot, properties {padding=<ltrb 5 5 5 5>}}
  }
  Overlay_JoinInfo {widgetClassPath=&Overlay, parent=GridPanel_JoinPrompt, widget=Overlay_JoinInfo, slot {class=GridSlot, properties {row=1}}}
  UIKit_Backplate {
   widgetClassPath=&WBP_UIKit_Backplate
   bIsVariable=true
   parent=Overlay_JoinInfo
   widget=UIKit_Backplate
   properties {brightness=Dark}
   slot {class=OverlaySlot, properties {horizontalAlignment=HAlign_Fill, verticalAlignment=VAlign_Fill}}
  }
  VerticalBox_0 {
   widgetClassPath=&VerticalBox
   parent=Overlay_JoinInfo
   widget=VerticalBox_0
   slot {class=OverlaySlot, properties {horizontalAlignment=HAlign_Fill, verticalAlignment=VAlign_Fill, padding=<ltrb 16 12 16 12>}}
  }
  Text_Primary {
   widgetClassPath=&CommonTextBlock
   bIsVariable=true
   parent=VerticalBox_0
   widget=Text_Primary
   properties {
    applyLineHeightToBottomLine=false
    bIsScrollingEnabled=false
    justification=Center
    text="[PH] Player {##}"
    font {
     bForceMonospaced=false
     bMaterialIsStencil=false
     fontMaterial=null
     fontObject=&HeadingNow
     letterSpacing=10
     monospacedWidth=1
     size=13.5
     skewAmount=0
     typefaceFontName=74 Regular
     outlineSettings {
      bApplyOutlineToDropShadows=false
      bMiteredCorners=false
      bSeparateFillAlpha=false
      outlineMaterial=null
      outlineSize=0
      outlineColor=<rgba 0 0 0 1>
     }
    }
    shadowOffset=<0 0>
    shapedTextOptions {bOverride_TextFlowDirection=true, bOverride_TextShapingMethod=false, textFlowDirection=Culture, textShapingMethod=Auto}
   }
   slot {class=VerticalBoxSlot, properties {horizontalAlignment=HAlign_Center}}
  }
  HorizontalBox_0 {
   widgetClassPath=&HorizontalBox
   parent=VerticalBox_0
   widget=HorizontalBox_0
   slot {class=VerticalBoxSlot, properties {horizontalAlignment=HAlign_Center, padding=<ltrb 0 4 0 0>}}
  }
  ActionWidget {
   widgetClassPath=&CommonActionWidget
   bIsVariable=true
   parent=HorizontalBox_0
   widget=ActionWidget
   slot {class=HorizontalBoxSlot, properties {horizontalAlignment=HAlign_Left, verticalAlignment=VAlign_Center}}
  }
  Text_Secondary {
   widgetClassPath=&CommonTextBlock
   bIsVariable=true
   parent=HorizontalBox_0
   widget=Text_Secondary
   properties {
    applyLineHeightToBottomLine=false
    bIsScrollingEnabled=false
    justification=Center
    mobileFontSizeMultiplier=1.3
    text=Press to join
    colorAndOpacity {colorUseRule=UseColor_Specified, specifiedColor=<rgba 0.5776 0.5776 0.5776 1>}
    font {
     bForceMonospaced=false
     bMaterialIsStencil=false
     fontMaterial=null
     fontObject=&HeadingNow
     letterSpacing=10
     monospacedWidth=1
     size=11.25
     skewAmount=0
     typefaceFontName=74 Regular
     outlineSettings {
      bApplyOutlineToDropShadows=false
      bMiteredCorners=false
      bSeparateFillAlpha=false
      outlineMaterial=null
      outlineSize=0
      outlineColor=<rgba 0 0 0 1>
     }
    }
    shadowOffset=<0 0>
    shapedTextOptions {bOverride_TextFlowDirection=true, bOverride_TextShapingMethod=false, textFlowDirection=Culture, textShapingMethod=Auto}
   }
   slot {class=HorizontalBoxSlot, properties {horizontalAlignment=HAlign_Left, verticalAlignment=VAlign_Center, padding=<ltrb 4 0 0 0>}}
  }
 ]
}
