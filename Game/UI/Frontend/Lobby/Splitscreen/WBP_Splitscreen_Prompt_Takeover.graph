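# Splitscreen lobby prompt widget, built on FortPedestalTakeControlWidget, that shows "Hold to take control" for the second player's take-control action.
# A dark, circle-cornered UIKit backplate sits behind a row holding the inherited, initially collapsed InputActionWidget and the label.
# The action widget uses the P2TakeControlV2 input row and a progress material driven through its Percentage parameter.
# PreConstruct sets that dynamic material's ShapeSwitch parameter to zero.
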WidgetBlueprint /Game/UI/Frontend/Lobby/Splitscreen/WBP_Splitscreen_Prompt_Takeover
graphs [EventGraph {nodes [
   Event PreConstruct()
   n1 = Get InputActionWidget() -> (InputActionWidget:Common Action Widget Object Reference)
   > SetScalarParameterValue(
    self=GetDynamicMaterial!(Brush:"Slate Brush Structure (by ref)"=Get ProgressMaterialBrush!(self:Common Action Widget Object Reference=@n1) -> (ProgressMaterialBrush:Slate Brush Structure)) -> (ReturnValue)
    ParameterName=ShapeSwitch
    Value=0
   )
  ]}]
variables []
widgets {
 info {inheritedWidgetCount=1, parentClass=/Script/FortniteUI.FortPedestalTakeControlWidget, rootWidgetClass=&Overlay, widgetCount=5}
 widgets [
  Overlay_2 {widgetClassPath=&Overlay, widget=Overlay_2}
  UIKit_Backplate {
   widgetClassPath=&WBP_UIKit_Backplate
   bIsVariable=true
   parent=Overlay_2
   widget=UIKit_Backplate
   properties {brightness=Dark, cornerRadius=Circle}
   slot {class=OverlaySlot, properties {horizontalAlignment=HAlign_Fill, verticalAlignment=VAlign_Fill}}
  }
  HorizontalBox_0 {
   widgetClassPath=&HorizontalBox
   parent=Overlay_2
   widget=HorizontalBox_0
   slot {class=OverlaySlot, properties {horizontalAlignment=HAlign_Fill, verticalAlignment=VAlign_Fill}}
  }
  InputActionWidget {
   widgetClassPath=&CommonActionWidget
   bInherited=true
   bIsVariable=true
   parent=HorizontalBox_0
   widget=InputActionWidget
   properties {
    designTimeKey=Gamepad_Special_Right
    progressMaterialParam=Percentage
    visibility=Collapsed
    iconRimBrush {
     bIsDynamicallyLoaded=false
     drawAs=NoDrawType
     imageType=NoImage
     mirroring=NoMirror
     resourceName=None
     resourceObject=null
     tiling=NoTile
     imageSize=<20 20>
     margin=<ltrb 0 0 0 0>
     outlineSettings {
      bUseBrushTransparency=false
      roundingType=HalfHeightRadius
      width=0
      color {colorUseRule=UseColor_Specified, specifiedColor=<rgba 0 0 0 0>}
      cornerRadii=<0 0 0 0>
     }
     tintColor {colorUseRule=UseColor_Specified, specifiedColor=<rgba 1 1 1 1>}
     uVRegion {bIsValid=false, max=<0 0>, min=<0 0>}
    }
    inputActions [{dataTable=&InputActionDataTable, rowName=P2TakeControlV2}]
    progressMaterialBrush {
     bIsDynamicallyLoaded=false
     drawAs=Image
     imageType=FullColor
     mirroring=NoMirror
     resourceName=None
     resourceObject=/FortUILibrary/UIKit/Buttons/MaterialInstances/CTAButtons/MI_UIKit_InputActionProgress
     tiling=NoTile
     imageSize=<48 48>
     margin=<ltrb 0 0 0 0>
     outlineSettings {
      bUseBrushTransparency=false
      roundingType=HalfHeightRadius
      width=0
      color {colorUseRule=UseColor_Specified, specifiedColor=<rgba 0 0 0 0>}
      cornerRadii=<0 0 0 0>
     }
     tintColor {colorUseRule=UseColor_Specified, specifiedColor=<rgba 1 1 1 1>}
     uVRegion {bIsValid=false, max=<0 0>, min=<0 0>}
    }
   }
   slot {class=HorizontalBoxSlot, properties {horizontalAlignment=HAlign_Center, verticalAlignment=VAlign_Center, padding=<ltrb 8 8 8 8>}}
  }
  Text_Control {
   widgetClassPath=&CommonTextBlock
   parent=HorizontalBox_0
   widget=Text_Control
   properties {
    justification=Center
    text=Hold to take control
    colorAndOpacity {colorUseRule=UseColor_Specified, specifiedColor=<rgba 0.75 0.75 0.75 1>}
    font {
     bForceMonospaced=false
     bMaterialIsStencil=false
     fontMaterial=null
     fontObject=&HeadingNow
     letterSpacing=10
     monospacedWidth=1
     size=11.25
     skewAmount=0
     typefaceFontName=74 Regular
     outlineSettings {
      bApplyOutlineToDropShadows=false
      bMiteredCorners=false
      bSeparateFillAlpha=false
      outlineMaterial=null
      outlineSize=0
      outlineColor=<rgba 0 0 0 1>
     }
    }
    shadowOffset=<0 0>
   }
   slot {class=HorizontalBoxSlot, properties {verticalAlignment=VAlign_Center, padding=<ltrb 0 0 16 0>}}
  }
 ]
}
